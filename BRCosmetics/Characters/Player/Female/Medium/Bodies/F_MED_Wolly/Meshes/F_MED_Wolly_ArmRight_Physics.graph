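# Physics asset for the right-arm mesh of the F_MED_Wolly female player body cosmetic, covering its dynamic knot and joystick bones.
# Box bodies shape the knot, joystick and both joystick bases, while a small capsule represents the knot's master bone.
# The knot is tightly limited against its master bone.
# The joystick base swings more widely against its own master, carrying a nearly rigid joystick as a two-link chain.
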PhysicsAsset /BRCosmetics/Characters/Player/Female/Medium/Bodies/F_MED_Wolly/Meshes/F_MED_Wolly_ArmRight_Physics
bodies {
 dyn_joystick [{shapeType=Box, extentX=2, extentY=0.5, extentZ=0.5, center=<1 0 0>}]
 dyn_joystickBase [{shapeType=Box, extentX=0.483, extentY=1.097, extentZ=1.097, center=<0.1161 0 0>}]
 dyn_knot [{shapeType=Box, extentX=10, extentY=2, extentZ=6, center=<5 0 0>}]
 dyn_mstr_joystickBase [{shapeType=Box, extentX=1.01, extentY=1.01, extentZ=1.01}]
 dyn_mstr_knot [{shapeType=Capsule, length=0.505, radius=0.505}]
}
constraints [
 {bone1Name=dyn_knot, bone2Name=dyn_mstr_knot, swing1LimitDegrees=5, swing2LimitDegrees=5, twistLimitDegrees=2.5}
 {bone1Name=dyn_joystickBase, bone2Name=dyn_mstr_joystickBase, swing1LimitDegrees=45, swing2LimitDegrees=45, twistLimitDegrees=15}
 {bone1Name=dyn_joystick, bone2Name=dyn_joystickBase, swing1LimitDegrees=2, swing2LimitDegrees=2, twistLimitDegrees=2}
]
