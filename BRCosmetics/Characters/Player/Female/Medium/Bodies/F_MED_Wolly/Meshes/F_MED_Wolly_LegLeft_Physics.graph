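# Physics asset for the left-leg mesh of the F_MED_Wolly player body cosmetic, covering its lower and upper strap bones.
# Each strap forms a separate chain: a small capsule on the master bone, followed by two box segments along the bone's X axis.
# Constraints link each segment to the one before it with tight swing and twist limits, so the straps can bend only slightly.
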PhysicsAsset /BRCosmetics/Characters/Player/Female/Medium/Bodies/F_MED_Wolly/Meshes/F_MED_Wolly_LegLeft_Physics
bodies {
 dyn_mstr_strapLower [{shapeType=Capsule, length=0.505, radius=0.505}]
 dyn_mstr_strapUpper [{shapeType=Capsule, length=0.505, radius=0.505}]
 dyn_strapLower1 [{shapeType=Box, extentX=3, extentY=1, extentZ=2, center=<1.5 0 0>}]
 dyn_strapLower2 [{shapeType=Box, extentX=2.5, extentY=1, extentZ=2, center=<1.25 0 0>}]
 dyn_strapUpper1 [{shapeType=Box, extentX=3, extentY=1, extentZ=2, center=<1.5 0 0>}]
 dyn_strapUpper2 [{shapeType=Box, extentX=2.5, extentY=1, extentZ=2, center=<1.25 0 0>}]
}
constraints [
 {bone1Name=dyn_strapLower1, bone2Name=dyn_mstr_strapLower, swing1LimitDegrees=5, swing2LimitDegrees=5, twistLimitDegrees=2}
 {bone1Name=dyn_strapLower2, bone2Name=dyn_strapLower1, swing1LimitDegrees=5, swing2LimitDegrees=5, twistLimitDegrees=2}
 {bone1Name=dyn_strapUpper1, bone2Name=dyn_mstr_strapUpper, swing1LimitDegrees=5, swing2LimitDegrees=5, twistLimitDegrees=2}
 {bone1Name=dyn_strapUpper2, bone2Name=dyn_strapUpper1, swing1LimitDegrees=5, swing2LimitDegrees=5, twistLimitDegrees=2}
]
