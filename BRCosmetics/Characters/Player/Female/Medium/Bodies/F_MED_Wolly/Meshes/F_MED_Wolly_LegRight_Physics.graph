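# Physics asset for the right leg mesh of the Wolly medium female player cosmetic, rigging its dynamic laces for physics.
# Small boxes on the lace bones form a chain from a master bone through a central knot bone, branching into left and right top-to-bottom strands.
# Every joint shares the same limited swing and tighter twist range.
# A single capsule on the calf twist bone, unconstrained to the laces, provides a leg collision shape.
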PhysicsAsset /BRCosmetics/Characters/Player/Female/Medium/Bodies/F_MED_Wolly/Meshes/F_MED_Wolly_LegRight_Physics
bodies {
 calf_twist_02_r [{shapeType=Capsule, length=23.5476, radius=5.621, center=<5.8035 3.4122 0.5517>, rotation=<rot 90 -0.5968 -5.5968>}]
 dyn_laces [{shapeType=Box, extentX=1, extentY=1, extentZ=3, center=<0.5 0 0>}]
 dyn_lacesBottom_l [{shapeType=Box, extentX=4.5, extentY=1, extentZ=4, center=<2.25 0 0>}]
 dyn_lacesBottom_r [{shapeType=Box, extentX=4.5, extentY=1, extentZ=4, center=<2.25 0 0>}]
 dyn_lacesTop_l [{shapeType=Box, extentX=5.5, extentY=1, extentZ=4, center=<2.75 0 0>}]
 dyn_lacesTop_r [{shapeType=Box, extentX=5.5, extentY=1, extentZ=4, center=<2.75 0 0>}]
 dyn_mstr_laces [{shapeType=Box, extentX=1.01, extentY=1.01, extentZ=1.01}]
}
constraints [
 {bone1Name=dyn_laces, bone2Name=dyn_mstr_laces, swing1LimitDegrees=30, swing2LimitDegrees=30, twistLimitDegrees=10}
 {bone1Name=dyn_lacesTop_l, bone2Name=dyn_laces, swing1LimitDegrees=30, swing2LimitDegrees=30, twistLimitDegrees=10}
 {bone1Name=dyn_lacesBottom_l, bone2Name=dyn_lacesTop_l, swing1LimitDegrees=30, swing2LimitDegrees=30, twistLimitDegrees=10}
 {bone1Name=dyn_lacesTop_r, bone2Name=dyn_laces, swing1LimitDegrees=30, swing2LimitDegrees=30, twistLimitDegrees=10}
 {bone1Name=dyn_lacesBottom_r, bone2Name=dyn_lacesTop_r, swing1LimitDegrees=30, swing2LimitDegrees=30, twistLimitDegrees=10}
]
