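# Secondary-motion AnimBP for the F_MED_Wolly character body, copying the parent mesh pose and correcting it.
# Update derives alphas from thigh rotations against the pelvis, driving constraints and additive offsets on thigh volume bones, with extra back offsets while skydiving.
# Constraints place strap, knot and joystick bones before three RigidBody simulations for legs and right arm.
# The laces constraint is unconnected, back-thigh alphas are unused, and ArmRight physics bases on upperarm_l.
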AnimBlueprint /BRCosmetics/Characters/Player/Female/Medium/Bodies/F_MED_Wolly/Meshes/F_MED_Wolly_AnimBP
target_skeleton=&/F_MED_Wolly_Skeleton
graphs [
 AnimGraph {nodes [
   "Constraint - dyn_mstr_laces (PP)"!(ComponentPose, Alpha=1, ConstraintWeights_0=0.8072, ConstraintWeights_1:float=0.1928) -> (Pose) {
    alphaBoolBlend {blendOption=Linear}
    boneToModify {boneName=dyn_mstr_laces}
    constraintSetup [{targetBone {boneName=calf_twist_02_r}}, {targetBone {boneName=calf_r}}]
    constraintWeights [1,1]
   }
   n1 = "Constraint - thigh_volume_bk_l (PP)"!(
    ComponentPose=Local To Component(LocalPose=Copy Pose From Mesh() {bUseAttachedParent=true})
    Alpha=1
    ConstraintWeights_0=0
    ConstraintWeights_1:float=1
   ) -> (Pose) {
    alphaBoolBlend {blendOption=Linear}
    boneToModify {boneName=thigh_volume_bk_l}
    constraintSetup [{targetBone {boneName=thigh_twist_01_l}}, {targetBone {boneName=pelvis}}]
    constraintWeights [1,1]
   }
   n2 = "Constraint - thigh_volume_bk_r (PP)"!(ComponentPose=@n1, Alpha=1, ConstraintWeights_0=0, ConstraintWeights_1:float=1) -> (Pose) {
    alphaBoolBlend {blendOption=Linear}
    boneToModify {boneName=thigh_volume_bk_r}
    constraintSetup [{targetBone {boneName=thigh_twist_01_r}}, {targetBone {boneName=pelvis}}]
    constraintWeights [1,1]
   }
   n3 = Get thigh_shape_fn_l() -> (thigh_shape_fn_l:double)
   n4 = "Constraint - thigh_volume_bk_l (PP)"(ComponentPose=@n2, Alpha=@n3, ConstraintWeights_0=0, ConstraintWeights_1=1) {
    alphaBoolBlend {blendOption=Linear}
    boneToModify {boneName=thigh_volume_bk_l}
    constraintSetup [{targetBone {boneName=thigh_twist_01_l}}, {targetBone {boneName=pelvis}}]
    constraintWeights [1,1]
   }
   n5 = Get thigh_shape_fn_r() -> (thigh_shape_fn_r:double)
   n6 = "Constraint - thigh_volume_bk_r (PP)"(ComponentPose=@n4, Alpha=@n5, ConstraintWeights_0=0, ConstraintWeights_1=1) {
    alphaBoolBlend {blendOption=Linear}
    boneToModify {boneName=thigh_volume_bk_r}
    constraintSetup [{targetBone {boneName=thigh_twist_01_r}}, {targetBone {boneName=pelvis}}]
    constraintWeights [1,1]
   }
   n7 = "Constraint - thigh_volume_fn_l (PP)"!(ComponentPose=@n6, Alpha=@n3, ConstraintWeights_0=0.55, ConstraintWeights_1:float=0.45) -> (Pose) {
    alphaBoolBlend {blendOption=Linear}
    boneToModify {boneName=thigh_volume_fn_l}
    constraintSetup [{targetBone {boneName=thigh_twist_01_l}}, {targetBone {boneName=pelvis}}]
    constraintWeights [1,1]
   }
   n8 = "Constraint - thigh_volume_fn_r (PP)"!(ComponentPose=@n7, Alpha=@n5, ConstraintWeights_0=0.55, ConstraintWeights_1:float=0.45) -> (Pose) {
    alphaBoolBlend {blendOption=Linear}
    boneToModify {boneName=thigh_volume_fn_r}
    constraintSetup [{targetBone {boneName=thigh_twist_01_r}}, {targetBone {boneName=pelvis}}]
    constraintWeights [1,1]
   }
   n9 = "Constraint - thigh_volume_lt_l1 (PP)"!(ComponentPose=@n8, Alpha=1, ConstraintWeights_0=0.5, ConstraintWeights_1:float=0.5) -> (Pose) {
    alphaBoolBlend {blendOption=Linear}
    boneToModify {boneName=thigh_volume_lt_l1}
    constraintSetup [{targetBone {boneName=thigh_twist_01_l}}, {targetBone {boneName=pelvis}}]
    constraintWeights [1,1]
   }
   n10 = "Constraint - thigh_volume_rt_r (PP)"!(ComponentPose=@n9, Alpha=1, ConstraintWeights_0=0.5, ConstraintWeights_1:float=0.5) -> (Pose) {
    alphaBoolBlend {blendOption=Linear}
    boneToModify {boneName=thigh_volume_rt_r}
    constraintSetup [{targetBone {boneName=thigh_twist_01_r}}, {targetBone {boneName=pelvis}}]
    constraintWeights [1,1]
   }
   n11 = Get thigh_shape_fn_l()
   n12 = "Transform (Modify) Bone - Bone: thigh_volume_fn_l"(ComponentPose=@n10, Alpha=@n11, Translation="2,-12,0", Rotation="0,0,0", Scale="1,1,1") {
    rotationMode=BMM_Additive
    rotationSpace=BCS_BoneSpace
    scaleMode=BMM_Additive
    scaleSpace=BCS_BoneSpace
    translationMode=BMM_Additive
    translationSpace=BCS_BoneSpace
    alphaBoolBlend {blendOption=Linear}
    boneToModify {boneName=thigh_volume_fn_l}
    translation=<2 -12>
   }
   n13 = Get thigh_shape_fn_r()
   n14 = "Transform (Modify) Bone - Bone: thigh_volume_fn_r"(ComponentPose=@n12, Alpha=@n13, Translation="-2,-10,0", Rotation="0,0,0", Scale="1,1,1") {
    rotationSpace=BCS_BoneSpace
    scaleMode=BMM_Additive
    scaleSpace=BCS_BoneSpace
    translationMode=BMM_Additive
    translationSpace=BCS_BoneSpace
    alphaBoolBlend {blendOption=Linear}
    boneToModify {boneName=thigh_volume_fn_r}
    translation=<-2 -10>
   }
   n15 = "Transform (Modify) Bone - Bone: thigh_volume_bk_l"(ComponentPose=@n14, Alpha=@n11, Translation="-8,0,-1", Rotation="0,0,0", Scale="1,1,1") {
    rotationMode=BMM_Additive
    rotationSpace=BCS_BoneSpace
    scaleMode=BMM_Additive
    scaleSpace=BCS_BoneSpace
    translationMode=BMM_Additive
    translationSpace=BCS_BoneSpace
    alphaBoolBlend {blendOption=Linear}
    boneToModify {boneName=thigh_volume_bk_l}
    translation {x=-8, z=-1}
   }
   n16 = "Transform (Modify) Bone - Bone: thigh_volume_bk_r"(ComponentPose=@n15, Alpha=@n13, Translation="8,0,-1", Rotation="0,0,0", Scale="1,1,1") {
    rotationSpace=BCS_BoneSpace
    scaleMode=BMM_Additive
    scaleSpace=BCS_BoneSpace
    translationMode=BMM_Additive
    translationSpace=BCS_BoneSpace
    alphaBoolBlend {blendOption=Linear}
    boneToModify {boneName=thigh_volume_bk_r}
    translation {x=8, z=-1}
   }
   n17 = Get bIsSkyDiving()
   n18 = "Transform (Modify) Bone - Bone: thigh_volume_bk_l"!(ComponentPose=@n16, bAlphaBoolEnabled:bool=@n17, Translation="-2,0,0", Rotation="0,0,0", Scale="1,1,1") -> (Pose) {
    alphaInputType=Bool
    rotationMode=BMM_Additive
    rotationSpace=BCS_BoneSpace
    scaleMode=BMM_Additive
    scaleSpace=BCS_BoneSpace
    translationMode=BMM_Additive
    translationSpace=BCS_BoneSpace
    alphaBoolBlend {blendOption=Linear}
    boneToModify {boneName=thigh_volume_bk_l}
    translation {x=-2}
   }
   n19 = "Transform (Modify) Bone - Bone: thigh_volume_bk_r"!(ComponentPose=@n18, bAlphaBoolEnabled:bool=@n17, Translation="-2,-3,0", Rotation="0,0,0", Scale="1,1,1") -> (Pose) {
    alphaInputType=Bool
    rotationSpace=BCS_BoneSpace
    scaleMode=BMM_Additive
    scaleSpace=BCS_BoneSpace
    translationMode=BMM_Additive
    translationSpace=BCS_BoneSpace
    alphaBoolBlend {blendOption=Linear}
    boneToModify {boneName=thigh_volume_bk_r}
    translation=<-2 -3>
   }
   n20 = "Transform (Modify) Bone - Bone: thigh_volume_lt_l1"(ComponentPose=@n19, Alpha=Get thigh_shape_lt_l() -> (thigh_shape_lt_l:double), Translation="0,-10,0", Rotation="0,40,0", Scale="1,1,1") {
    rotationMode=BMM_Additive
    rotationSpace=BCS_BoneSpace
    scaleMode=BMM_Additive
    scaleSpace=BCS_BoneSpace
    translationMode=BMM_Additive
    translationSpace=BCS_BoneSpace
    alphaBoolBlend {blendOption=Linear}
    boneToModify {boneName=thigh_volume_lt_l1}
    rotation {yaw=40}
    translation {y=-10}
   }
   n21 = "Transform (Modify) Bone - Bone: thigh_volume_rt_r"(ComponentPose=@n20, Alpha=Get thigh_shape_rt_r() -> (thigh_shape_rt_r:double), Translation="0,-10,0", Rotation="0,0,0", Scale="1,1,1") {
    rotationSpace=BCS_BoneSpace
    scaleMode=BMM_Additive
    scaleSpace=BCS_BoneSpace
    translationMode=BMM_Additive
    translationSpace=BCS_BoneSpace
    alphaBoolBlend {blendOption=Linear}
    boneToModify {boneName=thigh_volume_rt_r}
    translation {y=-10}
   }
   n22 = "Constraint - dyn_mstr_joystickBase (PPP)"!(ComponentPose=@n21, Alpha=1, ConstraintWeights_0=0.3985, ConstraintWeights_1:float=0.2156, ConstraintWeights_2:float=0.3858) -> (Pose) {
    alphaBoolBlend {blendOption=Linear}
    boneToModify {boneName=dyn_mstr_joystickBase}
    constraintSetup [{targetBone {boneName=lowerarm_twist_02_r}}, {targetBone {boneName=lowerarm_r}}, {targetBone {boneName=lowerarm_twist_01_r}}]
    constraintWeights [1,1,1]
   }
   n23 = "Constraint - dyn_mstr_knot (PPP)"!(ComponentPose=@n22, Alpha=1, ConstraintWeights_0=0.5494, ConstraintWeights_1:float=0.2315, ConstraintWeights_2:float=0.2191) -> (Pose) {
    alphaBoolBlend {blendOption=Linear}
    boneToModify {boneName=dyn_mstr_knot}
    constraintSetup [{targetBone {boneName=upperarm_twist_01_r}}, {targetBone {boneName=upperarm_r}}, {targetBone {boneName=upperarm_twist_02_r}}]
    constraintWeights [1,1,1]
   }
   n24 = "Constraint - dyn_mstr_strapUpper (PP)"!(ComponentPose=@n23, Alpha=1, ConstraintWeights_0=0.32, ConstraintWeights_1:float=0.68) -> (Pose) {
    alphaBoolBlend {blendOption=Linear}
    boneToModify {boneName=dyn_mstr_strapUpper}
    constraintSetup [{targetBone {boneName=calf_l}}, {targetBone {boneName=calf_twist_02_l}}]
    constraintWeights [1,1]
   }
   n25 = "Constraint - dyn_mstr_strapLower (PPP)"!(ComponentPose=@n24, Alpha=1, ConstraintWeights_0=0.1086, ConstraintWeights_1:float=0.0883, ConstraintWeights_2:float=0.8031) -> (Pose) {
    alphaBoolBlend {blendOption=Linear}
    boneToModify {boneName=dyn_mstr_strapLower}
    constraintSetup [{targetBone {boneName=calf_l}}, {targetBone {boneName=calf_twist_01_l}}, {targetBone {boneName=calf_twist_02_l}}]
    constraintWeights [1,1,1]
   }
   n26 = RigidBody(ComponentPose=@n25) {
    bClampLinearTranslationLimitToRefPose=true
    bOverrideWorldGravity=true
    overridePhysicsAsset=&/F_MED_Wolly_LegLeft_Physics
    simulationSpace=BaseBoneSpace
    alphaBoolBlend {blendOption=Linear}
    baseBoneRef {boneName=thigh_l}
    componentAppliedLinearAccClamp=<3000 3000 3000>
    componentLinearAccScale=<0.4 0.3 0.3>
    componentLinearVelScale=<0.4 0.3 0.3>
    overrideWorldGravity {z=-980}
   }
   n27 = RigidBody(ComponentPose=@n26) {
    bClampLinearTranslationLimitToRefPose=true
    bOverrideWorldGravity=true
    overridePhysicsAsset=&/F_MED_Wolly_LegRight_Physics
    simulationSpace=BaseBoneSpace
    alphaBoolBlend {blendOption=Linear}
    baseBoneRef {boneName=thigh_r}
    componentAppliedLinearAccClamp=<3000 3000 3000>
    componentLinearAccScale=<0.4 0.3 0.3>
    componentLinearVelScale=<0.4 0.3 0.3>
    overrideWorldGravity {z=-980}
   }
   n28 = RigidBody(ComponentPose=@n27) {
    bClampLinearTranslationLimitToRefPose=true
    bOverrideWorldGravity=true
    overridePhysicsAsset=&/F_MED_Wolly_ArmRight_Physics
    simulationSpace=BaseBoneSpace
    alphaBoolBlend {blendOption=Linear}
    baseBoneRef {boneName=upperarm_l}
    componentAppliedLinearAccClamp=<3000 3000 3000>
    componentLinearAccScale=<0.4 0.3 0.3>
    componentLinearVelScale=<0.4 0.3 0.3>
    overrideWorldGravity {z=-980}
   }
   Output Pose(Result=Component To Local(ComponentPose=@n28)) {layerGroup=None}
  ]}
 EventGraph {nodes [
   Event BlueprintInitializeAnimation()
   > Set Skel_Mesh(Skel_Mesh=GetOwnerSkeletalMeshComponent())
   Event BlueprintUpdateAnimation()
   n1 = Get Skel_Mesh()
   > Set thigh_shape_fn_l(thigh_shape_fn_l:double=GetAlphaFromReferenceTransformDelta(SkeletalMeshComponent=@n1, BoneName=thigh_l, BaseName=pelvis, InputRange=(X=0,Y=-110), TwistAxis="0,0,1")) -> (Output_Get:double)
   > Set thigh_shape_fn_r(thigh_shape_fn_r:double=GetAlphaFromReferenceTransformDelta(SkeletalMeshComponent=@n1, BoneName=thigh_r, BaseName=pelvis, InputRange=(X=0,Y=-110), TwistAxis="0,0,1")) -> (Output_Get:double)
   n2 = Get Skel_Mesh()
   > Set thigh_shape_bk_l(thigh_shape_bk_l:double=GetAlphaFromReferenceTransformDelta(SkeletalMeshComponent=@n2, BoneName=thigh_l, BaseName=pelvis, InputRange=(X=0,Y=60), TwistAxis="0,0,1")) -> (Output_Get:double)
   > Set thigh_shape_bk_r(thigh_shape_bk_r:double=GetAlphaFromReferenceTransformDelta(SkeletalMeshComponent=@n2, BoneName=thigh_r, BaseName=pelvis, InputRange=(X=0,Y=60), TwistAxis="0,0,1")) -> (Output_Get:double)
   n3 = Get Skel_Mesh()
   > Set thigh_shape_lt_l(thigh_shape_lt_l:double=GetAlphaFromReferenceTransformDelta(
     SkeletalMeshComponent=@n3
     BoneName=thigh_twist_01_l
     BaseName=pelvis
     DeltaComponentType=EulerY
     InputRange=(X=90,Y=0)
     OutputRange=(X=1,Y=0)
     TwistAxis="0,0,1"
    )) -> (Output_Get:double)
   > Set thigh_shape_rt_r(thigh_shape_rt_r:double=GetAlphaFromReferenceTransformDelta(
     SkeletalMeshComponent=@n3
     BoneName=thigh_twist_01_r
     BaseName=pelvis
     DeltaComponentType=EulerY
     InputRange=(X=90,Y=0)
     OutputRange=(X=1,Y=0)
     TwistAxis="0,0,1"
    )) -> (Output_Get:double)
  ]}
]
variables [
 Skel_Mesh {type=SkeletalMeshComponent}
 thigh_shape_bk_l {type=number}
 thigh_shape_bk_r {type=number}
 thigh_shape_fn_l {type=number}
 thigh_shape_fn_r {type=number}
 thigh_shape_lt_l {type=number}
 thigh_shape_rt_r {type=number}
]
