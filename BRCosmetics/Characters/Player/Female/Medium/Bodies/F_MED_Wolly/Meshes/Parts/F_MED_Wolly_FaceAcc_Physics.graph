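# Physics asset for the F_MED_Wolly female character's face accessory, defining simulated bodies for two ears and a trunk.
# Small capsules on the master ear and trunk-base bones anchor box bodies on the ears and eight trunk bones.
# The trunk segments form a chain of tightly limited joints, so bend accumulates along its length.
# Each ear hinges from its own anchor with a wide primary swing, narrower secondary swing and little twist.
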PhysicsAsset /BRCosmetics/Characters/Player/Female/Medium/Bodies/F_MED_Wolly/Meshes/Parts/F_MED_Wolly_FaceAcc_Physics
bodies {
 dyn_ear_l [{shapeType=Box, extentX=7.7963, extentY=6.793, extentZ=12.7942, center=<2.981 -0.6535 -0.7236>}]
 dyn_ear_r [{shapeType=Box, extentX=7.8298, extentY=7.0033, extentZ=12.7445, center=<2.8446 -0.7758 0.6798>}]
 dyn_mstr_ear_l [{shapeType=Capsule, length=0.505, radius=0.505}]
 dyn_mstr_ear_r [{shapeType=Capsule, length=0.505, radius=0.505}]
 dyn_mstr_trunkBase [{shapeType=Capsule, length=0.505, radius=0.505}]
 dyn_trunk1 [{shapeType=Box, extentX=4, extentY=5, extentZ=7, center=<2 0 0>}]
 dyn_trunk2 [{shapeType=Box, extentX=4, extentY=6, extentZ=7, center=<2 0 0>}]
 dyn_trunk3 [{shapeType=Box, extentX=4, extentY=5, extentZ=6, center=<2 0 0>}]
 dyn_trunk4 [{shapeType=Box, extentX=4, extentY=4, extentZ=5, center=<2 0 0>}]
 dyn_trunk5 [{shapeType=Box, extentX=3, extentY=4, extentZ=5, center=<1.5 0 0>}]
 dyn_trunk6 [{shapeType=Box, extentX=3, extentY=4, extentZ=5, center=<1.5 0 0>}]
 dyn_trunk7 [{shapeType=Box, extentX=2, extentY=4, extentZ=5, center=<1 0 0>}]
 dyn_trunk8 [{shapeType=Box, extentX=3, extentY=4, extentZ=4, center=<1.5 0 0>}]
}
constraints [
 {bone1Name=dyn_trunk1, bone2Name=dyn_mstr_trunkBase, swing1LimitDegrees=5, swing2LimitDegrees=5, twistLimitDegrees=5}
 {bone1Name=dyn_trunk2, bone2Name=dyn_trunk1, swing1LimitDegrees=5, swing2LimitDegrees=5, twistLimitDegrees=5}
 {bone1Name=dyn_trunk3, bone2Name=dyn_trunk2, swing1LimitDegrees=5, swing2LimitDegrees=5, twistLimitDegrees=5}
 {bone1Name=dyn_trunk4, bone2Name=dyn_trunk3, swing1LimitDegrees=5, swing2LimitDegrees=5, twistLimitDegrees=5}
 {bone1Name=dyn_trunk5, bone2Name=dyn_trunk4, swing1LimitDegrees=5, swing2LimitDegrees=5, twistLimitDegrees=5}
 {bone1Name=dyn_trunk6, bone2Name=dyn_trunk5, swing1LimitDegrees=5, swing2LimitDegrees=5, twistLimitDegrees=5}
 {bone1Name=dyn_trunk7, bone2Name=dyn_trunk6, swing1LimitDegrees=5, swing2LimitDegrees=5, twistLimitDegrees=5}
 {bone1Name=dyn_trunk8, bone2Name=dyn_trunk7, swing1LimitDegrees=5, swing2LimitDegrees=5, twistLimitDegrees=5}
 {bone1Name=dyn_ear_l, bone2Name=dyn_mstr_ear_l, swing1LimitDegrees=45, swing2LimitDegrees=15, twistLimitDegrees=5}
 {bone1Name=dyn_ear_r, bone2Name=dyn_mstr_ear_r, swing1LimitDegrees=45, swing2LimitDegrees=15, twistLimitDegrees=5}
]
